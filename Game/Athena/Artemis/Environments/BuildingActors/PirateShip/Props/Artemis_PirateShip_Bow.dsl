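; Pirate ship bow prop Blueprint that feeds its own world height to the pirate ship material parameter collection.
; Both the construction script and BeginPlay read the actor's Z location and write it to the GradientPosition scalar in MPC_PirateShip, only where cosmetic events can execute.
; Each write also prints the Z value with PrintString, apparently as debug output.
; The value is global to the collection and is not updated on tick.
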(event EventBeginPlay
  (bind _returnvalue (Transformation|GetActorLocation))
  (bind _z (.z _returnvalue))
  (Networking|CanExecuteCosmeticEvents
    (:True
      (Rendering|Material|SetScalarParameterValue "/Game/Environments/Sets/Pirate_Cove/Pirate_Ship/Materials/MaterialFunctions/MPC_PirateShip.MPC_PirateShip" "GradientPosition" _z :!class /Script/Engine.KismetMaterialLibrary)
      (Development|PrintString (Utilities|String|ToString(Float) _z)))))

(fn ConstructionScript
  (bind _returnvalue (Transformation|GetActorLocation))
  (bind _z (.z _returnvalue))
  (Networking|CanExecuteCosmeticEvents
    (:True
      (Rendering|Material|SetScalarParameterValue "/Game/Environments/Sets/Pirate_Cove/Pirate_Ship/Materials/MaterialFunctions/MPC_PirateShip.MPC_PirateShip" "GradientPosition" _z :!class /Script/Engine.KismetMaterialLibrary)
      (Development|PrintString (Utilities|String|ToString(Float) _z)))))
